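# Masked, lit beach ball material for Niagara mesh particles, with dynamic parameters driving custom color, explode, dissolve and metallic.
# Floored UVs split the ball for particle color and metallic, and CustomColor blends that color against dimmed vertex color.
# A chromatic noise texture drives both dithered dissolve and explode displacement along normals.
# Color is tinted by time of day, with view-facing emissive on sides away from the sky light.
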Material /DealershipSharedLibrary/Assets/Sets/BeachBalls/M_BeachBalls_Ball
graph {
 properties {bAutomaticallySetUsageInEditor=false, bUsedWithNiagaraMeshParticles=true, blendMode=BLEND_Masked, usedShadingModels=Default Lit}
 nodes [
  n1 = DynamicParameter() {paramNames [CustomColor,Explode,Dissolve,Metallic]}
  n2 = Floor(None=ComponentMask(None=TextureCoordinate() {uTiling=2}) {g=true, r=true})
  LinearInterpolate(A=@n1.Metallic, Alpha=@n2) {constB=0}
  n3 = SkyAtmosphereLightDirection()
  n4 = VertexNormalWS()
  n5 = OneMinus(None=Saturate(None=DotProduct(A=@n3, B=@n4)))
  n6 = Multiply(A=@n5, B=@n5)
  n7 = Multiply(A=OneMinus(None=Fresnel() {baseReflectFraction=0, exponent=1}), B=Multiply(A=@n6, B=@n6))
  n8 = LinearInterpolate(
   A=Multiply(A=VertexColor(), B=ScalarParameter() {defaultValue=0.4})
   B=LinearInterpolate(A=ParticleColor().RGB, Alpha=@n2) {constB=0.4}
   Alpha=@n1.CustomColor
  )
  MaterialFunctionCall(Emissive=Multiply(A=Multiply(A=@n7, B=@n8), B=ScalarParameter() {defaultValue=0.2, parameterName=Param_1}) {constB=0}) {materialFunction=&MF_FX_EmissiveAdjustments}
  n9 = TextureSample() {texture=/Game/Effects/Fort_Effects/Textures/Tile/T_Smooth_Tile_Chromatic}
  n10 = Add(A=OneMinus(None=@n9.R))
  n11 = Multiply(A=@n1.Dissolve) {constB=2}
  MaterialFunctionCall(Alpha Threshold=Saturate(None=Subtract(A=@n10, B=@n11))) {materialFunction=&DitherTemporalAA_Fortnite}
  Multiply(A=MaterialFunctionCall() {materialFunction=&Color_Time_of_Day}.Result, B=@n8)
  n12 = VertexNormalWS()
  Multiply(A=@n1.Explode, B=Multiply(A=Multiply(A=@n12) {constB=400}, B=Subtract(A=@n9.R)))
 ]
}
